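# Translucent unlit smoke aura material for Niagara mesh particles in SparksCosmetics.
# Panning trail mask and fire noise, offset per particle, form opacity shaped by vertical gradients and depth fade.
# Time-of-day purple tint, brightened by gradient and noise, drives emissive.
# World position offset sums SplineThicken width with a switchable, clamped, UV-weighted pull derived from particle position and Niagara parameters, using only X of the target.
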Material /SparksCosmetics/Items/Auras/Assets/NS_Aura_Smoke/M_Aura_Smoke
graph {
 properties {
  bUsedWithNiagaraMeshParticles=true
  blendMode=BLEND_Translucent
  pixelDepthOffsetMode=PDOM_Legacy
  shadingModel=MSM_Unlit
  usedShadingModels=Unlit
 }
 nodes [
  n1 = TextureCoordinate() {vTiling=0.5}
  n2 = Time()
  n3 = ParticleRandom()
  n4 = TextureSample(UVs=Panner(
    Coordinate=MaterialFunctionCall(XY=@n1) {materialFunction=&Swizzle}.YX
    Time=Add(A=@n2, B=@n3)
    Speed=Constant2Vector() {r=1}
   ) {speedX=0.25}) {texture=/Game/Effects/Fort_Effects/Textures/Smoke/T_Lobber_Trail_Mask}
  n5 = Multiply(A=Power(Base=@n4.G, Exp=Constant() {r=2}), B=Constant() {r=0.5}) {constB=0.3}
  n6 = TextureCoordinate() {uTiling=0.75, vTiling=0.4}
  n7 = Add(A=Time(), B=ParticleRandom())
  n8 = Constant2Vector() {g=0.454, r=0.1}
  n9 = OneMinus(None=TextureSample(UVs=Panner(Coordinate=@n6, Time=@n7, Speed=@n8) {speedY=0.1}) {samplerType=SAMPLERTYPE_LinearGrayscale, texture=/SparksCharacterCommon/Items/Auras/Textures/T_Mojo_Fire_Noise}.R)
  n10 = Multiply(A=@n4.G, B=@n9)
  n11 = Divide(A=Saturate(None=Add(A=@n5, B=@n10)), B=Constant() {r=0.5}) {constB=0.2}
  n12 = Multiply(
   A=Multiply(
    A=MaterialFunctionCall() {materialFunction=&LinearGradient}.VGradient
    B=Saturate(None=@n11)
   )
   B=SmoothStep(
    Min=Constant() {r=1}
    Max=Constant() {r=0.75}
    Value=MaterialFunctionCall() {materialFunction=&LinearGradient}.VGradient
   )
  )
  n13 = Constant() {r=3}
  DepthFade(Opacity=Saturate(None=Multiply(A=@n12, B=@n13) {constB=5})) {fadeDistanceDefault=10}
  n14 = Constant3Vector() {constant=<rgba 0.4729 0.21 0.8229 1>}
  n15 = MaterialFunctionCall(Color_Morning=@n14, Color_Day=@n14, Color_Evening=@n14, Color_Night=Constant3Vector() {constant=<rgba 0.4941 0.1001 0.5156 1>}) {materialFunction=&Color_Time_of_Day}
  n16 = Power(
   Base=MaterialFunctionCall() {materialFunction=&LinearGradient}.VGradient
   Exp=Constant() {r=3}
  ) {constExponent=4}
  MaterialFunctionCall(Emissive=LinearInterpolate(A=Multiply(A=@n15.Result) {constB=0.025}, B=@n15.Result, Alpha=Multiply(A=@n16, B=@n9))) {materialFunction=&MF_FX_EmissiveAdjustments}
  n17 = DynamicParameter() {desc=Parameters set by Niagara, parameterIndex=3, defaultValue=<rgba 0 1 1 1>, paramNames [X,Y,Z,Intensity]}
  n18 = MaterialFunctionCall(Float3=Subtract(
    A=MaterialFunctionCall(X=@n17.X, Y=@n17.X, Z=@n17.X) {materialFunction=&MakeFloat3}.Result
    B=ParticlePositionWS()
   )) {materialFunction=&BreakOutFloat3Components}
  n19 = AppendVector(A=@n18.R, B=@n18.R)
  n20 = ScalarParameter() {defaultValue=100, desc=Interpolate Clamp XY, group=WPO, parameterName=Interpolate XY Clamp, sortPriority=3}
  n21 = Multiply(A=@n20) {constB=-1}
  n22 = ScalarParameter() {defaultValue=2, desc=Interpolate Clamp Z, group=WPO, parameterName=Interpolate Clamp Z, sortPriority=4}
  n23 = Multiply(A=@n22) {constB=-1}
  n24 = ComponentMask(None=OneMinus(None=TextureCoordinate())) {g=true}
  n25 = ScalarParameter() {defaultValue=5, desc=Interpolate Power, group=WPO, parameterName=Interpolate Power, sortPriority=2}
  n26 = LinearInterpolate(
   B=AppendVector(A=Clamp(None=@n19, Min=@n21, Max=@n20) {maxDefault=1000, minDefault=-1000}, B=Clamp(None=@n18.B, Min=@n23, Max=@n22) {minDefault=-1})
   Alpha=Multiply(A=@n17.Intensity, B=Power(Base=@n24, Exp=@n25))
  )
  n27 = Constant()
  n28 = DynamicParameter() {defaultValue=<rgba -60 1 1 1>, paramNames [Width,Param2,Param3,Param4]}
  NamedRerouteDeclaration(None=Add(
    A=StaticSwitchParameter(True=@n26, False=@n27) {defaultValue=true, group=WPO, parameterName=Use Interpolate WPO, sortPriority=1}
    B=MaterialFunctionCall(WidthBase=@n28.Width, WidthTip=@n28.Width) {materialFunction=&SplineThicken}.WorldPositionOffset
   )) {name=WPO, nodeColor=<rgba 0 0 0 1>}
  NamedRerouteUsage()
 ]
}
