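# Unlit translucent Niagara sprite and mesh particle material drawing radial smoke for the smoke aura cosmetic.
# Polar-mapped smoke mask and fire noise pan with per-particle time offsets, sampled using original UV derivatives to avoid seam artifacts.
# Circular masks shape a hollow-centred falloff, eroded by the Dissolve dynamic parameter and softened by depth fade.
# Emissive blends dim to full time-of-day purple along a vertical gradient, then passes through shared FX emissive adjustments.
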Material /SparksCosmetics/Items/Auras/Assets/NS_Aura_Smoke/M_Aura_Smoke_Radial
graph {
 properties {
  bUsedWithNiagaraMeshParticles=true
  bUsedWithNiagaraSprites=true
  blendMode=BLEND_Translucent
  pixelDepthOffsetMode=PDOM_Legacy
  shadingModel=MSM_Unlit
  usedShadingModels=Unlit
 }
 nodes [
  n1 = MaterialFunctionCall(XY=Multiply(
    A=Constant2Vector() {g=0.5, r=3}
    B=MaterialFunctionCall() {materialFunction=&VectorToRadialValue}.Radial Coordinates
   )) {materialFunction=&Swizzle}
  n2 = Add(A=Time(), B=ParticleRandom())
  n3 = Constant2Vector() {r=-1}
  n4 = TextureCoordinate()
  n5 = TextureCoordinate()
  n6 = Constant() {r=0.5}
  n7 = Constant() {r=0.5}
  n8 = Constant() {r=0.5}
  n9 = Multiply(
   A=TextureSample!(
    UVs=Panner(Coordinate=@n1.YX, Time=@n2, Speed=@n3) {speedX=0.25}
    Tex
    "DDX(UVs)"=DDX(Value=@n4)
    "DDY(UVs)"=DDY(Value=@n4)
    Apply View MipBias
   ) {mipValueMode=TMVM_Derivative, texture=/Game/Effects/Fort_Effects/Textures/Smoke/T_Lobber_Trail_Mask}.G
   B=Power(Base=SphereMask(A=@n5, B=@n6, Radius=@n7, Hardness=@n8) {hardnessPercent=0}) {constExponent=1}
  )
  n10 = Power(Base=@n9, Exp=Constant() {r=2})
  n11 = Constant() {r=0.5}
  n12 = MaterialFunctionCall() {materialFunction=&VectorToRadialValue}
  n13 = Constant2Vector() {g=0.25, r=4}
  n14 = Time()
  n15 = ParticleRandom()
  n16 = TextureCoordinate()
  n17 = TextureSample(
   UVs=Panner(Coordinate=Multiply(A=@n12.Radial Coordinates, B=@n13), Time=Add(A=@n14, B=@n15), Speed=Constant2Vector() {g=-0.75}) {speedY=0.1}
   "DDX(UVs)"=DDX(Value=@n16)
   "DDY(UVs)"=DDY(Value=@n16)
  ) {mipValueMode=TMVM_Derivative, samplerType=SAMPLERTYPE_LinearGrayscale, texture=/SparksCharacterCommon/Items/Auras/Textures/T_Mojo_Fire_Noise}
  n18 = NamedRerouteDeclaration(None=OneMinus(None=@n17.R)) {name=NoiseTexture, nodeColor=<rgba 0 0 0 1>}
  n19 = Saturate(None=Add(A=Multiply(A=@n10, B=@n11) {constB=0.3}, B=Multiply(A=@n9, B=@n18)))
  n20 = DynamicParameter() {defaultValue=<rgba 0 1 1 1>, paramNames [Dissolve,Param2,Param3,Param4]}
  n21 = SphereMask(A=TextureCoordinate(), B=Constant() {r=0.5}, Radius=Constant() {r=0.3}, Hardness=Constant()) {hardnessPercent=0}
  n22 = Multiply(A=Saturate(None=Subtract(A=@n19, B=@n20.Dissolve) {constB=0.1}), B=Power(Base=OneMinus(None=@n21)))
  n23 = Constant() {r=3}
  DepthFade(Opacity=Saturate(None=Multiply(A=@n22, B=@n23) {constB=5})) {fadeDistanceDefault=10}
  n24 = Constant3Vector() {constant=<rgba 0.4729 0.21 0.8229 1>}
  n25 = MaterialFunctionCall(Color_Morning=@n24, Color_Day=@n24, Color_Evening=@n24, Color_Night=Constant3Vector() {constant=<rgba 0.4941 0.1001 0.5156 1>}) {materialFunction=&Color_Time_of_Day}
  n26 = Power(
   Base=MaterialFunctionCall() {materialFunction=&LinearGradient}.VGradient
   Exp=Constant() {r=3}
  ) {constExponent=4}
  n27 = NamedRerouteUsage()
  MaterialFunctionCall(Emissive=LinearInterpolate(A=Multiply(A=@n25.Result) {constB=0.025}, B=@n25.Result, Alpha=Multiply(A=@n26, B=@n27))) {materialFunction=&MF_FX_EmissiveAdjustments}
 ]
}
