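# Translucent, unlit, two-sided Niagara particle material for the purple smoke wisps of a cosmetic aura.
# A streak texture, panned over time with a per-particle random offset and swizzled UVs, is masked by two generated bands, then particle alpha and depth fade shape opacity.
# Time-of-day color brightens from dim to full along a powered vertical gradient multiplied by the streaks.
# A dynamic parameter sets SplineThicken width.
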Material /SparksCosmetics/Items/Auras/Assets/NS_Aura_Smoke/M_Aura_Smoke_Wisps
graph {
 properties {
  bUsedWithNiagaraMeshParticles=true
  bUsedWithNiagaraSprites=true
  blendMode=BLEND_Translucent
  pixelDepthOffsetMode=PDOM_Legacy
  shadingModel=MSM_Unlit
  twoSided=true
  usedShadingModels=Unlit
 }
 nodes [
  n1 = Time()
  n2 = Constant() {r=0.75}
  n3 = Panner(Coordinate=TextureCoordinate() {uTiling=0.2, vTiling=0.75}, Time=Add(A=Multiply(A=@n1, B=@n2), B=ParticleRandom())) {speedY=1}
  n4 = Saturate(None=TextureSample(UVs=MaterialFunctionCall(XY=@n3) {materialFunction=&Swizzle}.YX) {texture=/Game/Effects/Fort_Effects/Textures/Energy/T_OutpostShockwaveStreaks_02}.R)
  n5 = Divide(A=Subtract(A=@n4, B=Constant() {r=0.2}) {constB=0.1}, B=Constant() {r=0.75}) {constB=0.2}
  n6 = MaterialFunctionCall!(Width=Constant() {r=0.5}, Sharpness, Offset, Direction Switch=StaticBool() {value=true}, Compare, Input Coordinates) {materialFunction=/Engine/Functions/Engine_MaterialFunctions03/Procedurals/GeneratedBand}
  n7 = MaterialFunctionCall(Width=Constant() {r=0.3}) {materialFunction=&GeneratedBand}
  n8 = Multiply(A=Multiply(A=Saturate(None=@n5), B=Multiply(A=@n6.Result, B=@n7.Result)), B=Constant() {r=3}) {constB=5}
  DepthFade(Opacity=Multiply(A=ParticleColor().A, B=Saturate(None=@n8))) {fadeDistanceDefault=5}
  MaterialFunctionCall() {materialFunction=&RadialGradientExponential}
  n9 = Constant3Vector() {constant=<rgba 0.4729 0.21 0.8229 1>}
  n10 = MaterialFunctionCall(Color_Morning=@n9, Color_Day=@n9, Color_Evening=@n9, Color_Night=Constant3Vector() {constant=<rgba 0.4941 0.1001 0.5156 1>}) {materialFunction=&Color_Time_of_Day}
  n11 = Power(
   Base=MaterialFunctionCall() {materialFunction=&LinearGradient}.VGradient
   Exp=Constant() {r=3}
  ) {constExponent=4}
  MaterialFunctionCall(Emissive=LinearInterpolate(A=Multiply(A=@n10.Result) {constB=0.025}, B=@n10.Result, Alpha=Multiply(A=@n11, B=@n4))) {materialFunction=&MF_FX_EmissiveAdjustments}
  n12 = DynamicParameter() {paramNames [SplineWidth,Param2,Param3,Param4]}
  MaterialFunctionCall(WidthBase=@n12.SplineWidth, WidthTip=@n12.SplineWidth) {materialFunction=&SplineThicken}
 ]
}
